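# Niagara effect for the SulfurDean_Hear player emote that flashes brief, dark, downward speed lines, apparently accenting a shirt motion.
# The active speed_lines_down emitter bursts sprites from a raised ring, fires them down with drag, and animates alpha and randomized stretch over their short life.
# A front-end float modifier shifts camera offset between gameplay and front end.
# The mesh-based rings emitter and a User.WorkoutSelect size scale are disabled.
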NiagaraSystem /BRCosmetics/Animation/Game/MainPlayer/Emotes/SulfurDean_Hear/NS_Emote_SulfurDean_Hear_shirtLines_down
dependencies {
 usedDataInterfaces [&NiagaraDataInterfaceCurve, &NiagaraDataInterfaceVector2DCurve, &NiagaraDataInterfaceArrayFloat]
 usedDynamicInputs [&MakeFloatFromVector2d]
 usedModules [&SystemState]
}
emitters {
 rings {
  bEnabled=false
  emitterUpdateScript [
   &EmitterState(Life Cycle Mode=::NewEnumerator1, Loop Behavior=::NewEnumerator1, Loop Duration=1, UseLoopDelay=false)
   &SpawnBurst_Instantaneous(Use Spawn Probability=false, UseLoopCountLimit=false)
  ]
  particleSpawnScript [
   &InitializeParticle(
    Color=<rgba 0.1329 0.2874 0.3663 0.5>
    Lifetime=0.2
    Mesh Scale=<0.2 0.2 0.15>
    Mesh Scale Mode=::NewEnumerator4
    Position Offset=<0 0 50>
    UsePositionOffset=true
    Write Material Random=true
   )
   &InitialMeshOrientation(
    Mesh Orientation Mode=::NewEnumerator1
    Rotation=&RandomRangeVector(Minimum=<0 0 0>, Maximum=<0 0 1>)
    Use Rotation Vector=true
   ) {enabled=false}
   &AddVelocity(Velocity=<0 0 -300>)
   &ApplyInitialForces()
  ]
  particleUpdateScript [
   &ParticleState()
   &ScaleMeshSize(Scale Factor=&VectorFromFloat(Value=&FloatFromCurve(FloatCurve=DataInterface() {propertyValues {Curve {
         defaultValue=0
         keys [{interpMode=RCIM_Cubic, tangentMode=RCTM_User, time=0, value=0.4}, {interpMode=RCIM_Cubic, tangentMode=RCTM_User, time=1, value=1}]
        }}}))) {enabled=false}
   &ScaleColor(
    Scale Alpha=&FloatFromCurve(FloatCurve=DataInterface() {propertyValues {Curve {
        defaultValue=0
        keys [
         {arriveTangent=-0.0576, interpMode=RCIM_Cubic, leaveTangent=-0.0576, tangentMode=RCTM_User, time=0.1333, value=0.0158}
         {arriveTangent=0.1271, interpMode=RCIM_Cubic, leaveTangent=0.1271, tangentMode=RCTM_User, time=0.5, value=0.9771}
         {arriveTangent=-6.2895, tangentMode=RCTM_User, time=0.85, value=-0.0092}
        ]
       }}})
    ScaleA=true
    ScaleRGB=true
   )
   &DynamicMaterialParameters(
    Index 0 Param 1=&FloatFromCurve(FloatCurve=DataInterface() {propertyValues {Curve {keys [<at 0 1.5>,<at 0.5 1.5>,<at 1 0>]}}})
    Index 0 Param 2=0.5
    Index 0 Param 4=1
    Module.Index0Float0Write=true
    Module.Index0Float1Write=true
    Module.Index0Float2Write=true
    Module.Index0Float3Write=true
   )
   &GravityForce() {enabled=false}
   &Drag(
    Drag=&FloatFromCurve(FloatCurve=DataInterface() {propertyValues {Curve {defaultValue=1, keys [{tangentMode=RCTM_User, time=0, value=0}, {tangentMode=RCTM_User, time=0.7698, value=5}]}}})
    Use Linear Drag=true
    Use Rotational Drag=true
   )
   &SolveForcesAndVelocity(Clamp Velocity=false, Limit Acceleration=false)
  ]
  rendererClasses [&NiagaraMeshRendererProperties]
  renderers [
   {rendererClass=&NiagaraMeshRendererProperties, rendererIndex=0}
   {rendererClass=&NiagaraMeshRendererProperties, rendererIndex=1}
   {rendererClass=&NiagaraMeshRendererProperties, rendererIndex=2}
   {rendererClass=&NiagaraMeshRendererProperties, rendererIndex=3}
  ]
 }
 speed_lines_down {
  emitterUpdateScript [
   &EmitterState()
   &SpawnBurst_Instantaneous(Spawn Count=10, Use Spawn Probability=false, UseLoopCountLimit=false)
  ]
  particleSpawnScript [
   &InitializeParticle(
    Color=&RandomRangeLinearColor(Minimum=<rgba 0 0 0 10>, Maximum=<rgba 0.0073 0.0073 0.0073 15>)
    Lifetime=0.12
    Position Offset=<0 0 50>
    Sprite Size=<16.6667 85>
    Sprite Size Mode=::NewEnumerator4
    UsePositionOffset=true
    Write Material Random=true
   )
   &CameraOffset(Camera Offset Amount=&FN_FrontEndFloatModifier(Gameplay=0, Front End=-30))
   &ShapeLocation(MaskSpawnGroup=false, Non Uniform Scale=<1 1 0>, Override Seed=false, Ring Radius=20, Shape Primitive=::NewEnumerator4, UseOwnerScale=true)
   &AddVelocity(Velocity=<0 0 -500>)
  ]
  particleUpdateScript [
   &ParticleState()
   &Drag(Drag=5, Use Linear Drag=true, Use Rotational Drag=false)
   &SolveForcesAndVelocity(Clamp Velocity=false, Limit Acceleration=false)
   &ScaleColor(
    Scale Alpha=&FloatFromCurve(FloatCurve=DataInterface() {propertyValues {Curve {keys [
         {interpMode=RCIM_Cubic, tangentMode=RCTM_User, time=0.1552, value=0}
         {interpMode=RCIM_Cubic, tangentMode=RCTM_User, time=0.5, value=1}
         {tangentMode=RCTM_User, time=1, value=0}
        ]}}})
    ScaleA=true
    ScaleRGB=true
   )
   &ScaleSpriteSize(
    Non-Uniform Curve Scale=&RandomRangeVector2D(Minimum=<0.7 0.5>, Maximum=<1.3 1>)
    Non-Uniform Curve Sprite Scale=DataInterface() {propertyValues {
      XCurve {keys [{tangentMode=RCTM_User, time=0, value=0}, {tangentMode=RCTM_User, time=1, value=1.5}]}
      YCurve {keys [{tangentMode=RCTM_User, time=0, value=0}, {tangentMode=RCTM_User, time=1, value=-1}]}
     }}
    Scale Sprite Size Mode=::NewEnumerator3
   )
   &ScaleSpriteSize(
    Scale Sprite Size Mode=::NewEnumerator0
    Uniform Scale Factor=&SelectFloatFromArray(
     Direct Array Index=&MakeIntFromFloat(Float=$User.WorkoutSelect)
     Float Selection Array=DataInterface() {propertyValues {FloatData [1,1,1,0.8]}}
    )
   ) {enabled=false, module=ScaleSpriteSize001}
   /SetVariables(Particles.New Position:NiagaraPosition=&Add_Vector(
     A=Linked() {linkedVariable {name=Particles.Position, type=NiagaraPosition}}
     B=Linked() {linkedVariable {name=Engine.Emitter.SimulationPosition, type=NiagaraPosition}}
    )) {enabled=false, isSetParameters=true}
  ]
  rendererClasses [&NiagaraSpriteRendererProperties]
  renderers [{rendererClass=&NiagaraSpriteRendererProperties}]
 }
}
summary {
 systemName=NS_Emote_SulfurDean_Hear_shirtLines_down
 emitters [
  {emitterName=speed_lines_down, rendererClasses [&NiagaraSpriteRendererProperties]}
  {bEnabled=false, emitterName=rings, rendererClasses [&NiagaraMeshRendererProperties]}
 ]
 userVariables [
  User.bIsFrontEnd {type=NiagaraBool}
  User.bIsFrontEndPreview {type=NiagaraBool}
 ]
}
