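# Niagara effect for the drink part of the SulfurDean_Hear player emote: liquid splashes and short speed-line flicks attached to skeletal mesh sockets.
# New_Splashes emits curve-timed SubUV sprites from one socket, culls them outside a box around a second, and blends local/world space by owner velocity.
# Two Speed_Lines emitters fire mirrored bursts, repeating or at set times, with front-end camera offsets.
# The ribbon trail and second splash emitter are disabled.
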NiagaraSystem /BRCosmetics/Animation/Game/MainPlayer/Emotes/SulfurDean_Hear/NS_SulfurDean_Hear_Drink
dependencies {
 usedDataInterfaces [
  &NiagaraDataInterfaceCurve
  &NiagaraDataInterfaceSpriteRendererInfo
  &NiagaraDataInterfaceColorCurve
  &NiagaraDataInterfaceVector2DCurve
  &NiagaraDataInterfaceArrayFloat
  &NiagaraDataInterfaceSkeletalMesh
 ]
 usedDynamicInputs [&MakeFloatFromVector2d]
 usedModules [
  &SystemState
  &self:SystemSpawnScript.SystemScriptSource.SystemScriptGraph.NiagaraNodeAssignment_1.SetVariables_#1
  &BlendSpaceHelper
 ]
}
emitters {
 BladeTrailDark {
  bEnabled=false
  emitterUpdateScript [
   &EmitterState(Life Cycle Mode=::NewEnumerator1, Loop Behavior=::NewEnumerator1, Loop Delay=1, Loop Duration=0.5, UseLoopDelay=true)
   &FN_SpawnAnimTrail(
    Spawn Interval=0
    Spawn Mode:ENiagaraSkeletalSpawning=::NewEnumerator0
    Switch on ENiagara_AnimTrail_Source:ENiagara_AnimTrail_Source=::NewEnumerator1
   )
  ]
  particleSpawnScript [
   &FN_InitializeAnimTrailBoneIndecies(Second Bone or Socket Index=0)
   &FN_InitializeAnimTrail(
    AnimTrailWidth=&Multiply_Float(A=7, B=1)
    Color=&MultiplyLinearColor(Linear Color=<rgba 0.074 0.3229 0.2626 3>, Scale Factor=<1 1 1 1>)
    Lifetime=0.25
    SkeletalMesh=$System.Skeletal Mesh
   )
  ]
  particleUpdateScript [
   &FN_UpdateAnimTrail(SkeletalMesh=$System.Skeletal Mesh) {module=FN_UpdateAnimTrail001}
   &ScaleColor(
    Scale Alpha=&Multiply_Float(
     A=&FloatFromCurve(
      FloatCurve=DataInterface() {propertyValues {Curve {keys [
          {arriveTangent=-2.003, interpMode=RCIM_Cubic, leaveTangent=-2.003, tangentMode=RCTM_User, time=0, value=1}
          {arriveTangent=-1, interpMode=RCIM_Cubic, leaveTangent=-1, time=0.5, value=0.2}
          {interpMode=RCIM_Cubic, tangentMode=RCTM_User, time=1, value=0}
         ]}}}
      Scale Curve=&FN_Select_Float_from_Time_Of_Day(Morning=0.3, Day=0.35, Evening=0.25, Night=0.12)
     )
     B=3
    )
    Scale RGB=&Multiply_Vector(
     A=&MakeVectorFromLinearColorRGB(Color=&ColorFromCurve(ColorCurve=DataInterface() {propertyValues {
         AlphaCurve {keys []}
         BlueCurve {keys [{arriveTangent=-1.3253, interpMode=RCIM_Cubic, time=0, value=1}, <at 0.3953 0.84>, <at 1 0.6056>]}
         GreenCurve {keys [{arriveTangent=-1.3253, interpMode=RCIM_Cubic, time=0, value=1}, <at 0.3953 0.8898>, <at 1 0.7>]}
         RedCurve {keys [{arriveTangent=-1.3253, interpMode=RCIM_Cubic, time=0, value=0.94}, <at 0.3953 0.7739>, <at 1 0.4819>]}
        }}))
     B=&MakeVectorFromLinearColorRGB(Color=&ColorFromCurve(
       ColorCurve=DataInterface() {propertyValues {
         AlphaCurve {keys [<at 0 1>,<at 0.125 1>,<at 0.25 1>,<at 0.375 1>,<at 0.5 1>,<at 0.625 1>,<at 0.75 1>,<at 0.875 1>,<at 1 1>]}
         BlueCurve {keys [<at 0.125 0.7835>,<at 0.166 0.6172>,<at 0.208 0.5395>,<at 0.25 0.6172>,<at 0.291 0.6939>,<at 0.333 0.7379>,<at 0.375 0.7835>,<at 0.416 0.7991>,<at 0.458 0.8148>,<at 0.5 0.8308>,<at 0.541 0.8148>,<at 0.583 0.7991>,<at 0.625 0.7991>,<at 0.666 0.7529>,<at 0.708 0.7084>,<at 0.75 0.5711>,<at 0.791 0.4735>,<at 0.833 0.3231>,<at 0.8778 0.8388>]}
         GreenCurve {keys [<at 0.125 0.5583>,<at 0.166 0.5333>,<at 0.208 0.5271>,<at 0.25 0.6376>,<at 0.291 0.7529>,<at 0.333 0.7835>,<at 0.375 0.8148>,<at 0.416 0.7991>,<at 0.458 0.8148>,<at 0.5 0.8469>,<at 0.541 0.7991>,<at 0.583 0.7991>,<at 0.625 0.8148>,<at 0.666 0.7991>,<at 0.708 0.7529>,<at 0.75 0.5711>,<at 0.791 0.5029>,<at 0.833 0.5029>,<at 0.8778 0.5333>]}
         RedCurve {keys [<at 0.125 0.2831>,<at 0.166 0.7084>,<at 0.208 0.6867>,<at 0.25 0.8148>,<at 0.291 0.8879>,<at 0.333 0.8963>,<at 0.375 0.9131>,<at 0.416 0.8714>,<at 0.458 0.8714>,<at 0.5 0.9131>,<at 0.541 0.8632>,<at 0.583 0.8714>,<at 0.625 0.9131>,<at 0.666 0.8963>,<at 0.708 0.8714>,<at 0.75 0.7529>,<at 0.791 0.7529>,<at 0.833 0.8469>,<at 0.8778 0.1878>]}
        }}
       CurveIndex=Linked() {linkedVariable {name=NPC.FortniteNPC.TimeOfDay_0-1, type=NiagaraFloat}}
      ))
    )
    ScaleA=true
    ScaleRGB=false
   )
   &ScaleRibbonWidth(Ribbon Width Scale=&FloatFromCurve(FloatCurve=DataInterface() {propertyValues {Curve {keys [<at 0 1>,<at 1 0>]}}}))
  ]
  rendererClasses [&NiagaraRibbonRendererProperties]
  renderers [{rendererClass=&NiagaraRibbonRendererProperties}]
 }
 New_Splashes {
  emitterUpdateScript [
   &EmitterState(Life Cycle Mode=::NewEnumerator1, Loop Behavior=::NewEnumerator1, Loop Duration=1, UseLoopDelay=false)
   &SpawnRate(
    SpawnRate=&FloatFromCurve(
     FloatCurve=DataInterface() {propertyValues {Curve {
        defaultValue=1
        keys [
         {arriveTangent=-2.0784, interpMode=RCIM_Cubic, leaveTangent=-2.0784, tangentMode=RCTM_User, time=0.2, value=1}
         {arriveTangent=-3.1211, interpMode=RCIM_Cubic, leaveTangent=-3.1211, time=0.5179, value=0.5912}
         {arriveTangent=0.0533, interpMode=RCIM_Cubic, leaveTangent=0.0533, tangentMode=RCTM_User, time=0.5279, value=-0.0235}
        ]
       }}}
     CurveIndex=$Emitter.NormalizedLoopAge
     Scale Curve=130
    )
    Use Spawn Probability=false
   )
  ]
  particleSpawnScript [
   &SkeletalMeshLocation(
    MaskSpawnGroup=false
    Mesh Sampling Type=::NewEnumerator5
    Orientation Sampling=::NewEnumerator1
    Position Sampling=::NewEnumerator2
    Sampled Position Offset=<-1 0 0.5>
    Sampled Position Offset Coordinate Space:ENiagaraSkelMeshTransforms=::NewEnumerator2
    Skeletal Mesh=$System.Skeletal Mesh
    Socket Sampling Mode=::NewEnumerator5
   )
   &InitializeParticle(
    Color=<rgba 0.2248 1 0.6955 1>
    Lifetime Max=0.4
    Lifetime Min=0.3
    Lifetime Mode=::NewEnumerator1
    Sprite Rotation Mode=::NewEnumerator2
    Sprite Size Mode=::NewEnumerator1
    Sprite UV Mode=::NewEnumerator3
    Uniform Sprite Size Max=30
    Uniform Sprite Size Min=20
    UsePositionOffset=false
    Write Material Random=true
   )
   &SubUVAnimation(
    "Sprite Renderer "=DataInterface() {propertyValues {SpriteRenderer=&self:New_Splashes_0.NiagaraSpriteRendererProperties_1}}
    SubUV Animation Mode:ENiagara_SubUVLookupModeV2=::NewEnumerator2
    UseEndFrame=false
    UseStartFrame=false
   )
   &ShapeLocation(
    MaskSpawnGroup=false
    Override Seed=false
    Shape Origin=$Particles.SkeletalMeshLocation.SampledPosition
    Sphere Radius=0.1
    UseOwnerScale=true
   ) {module=ShapeLocation001}
   &AddVelocity(
    Additional Quaternion Rotation=false
    Cone Angle=0
    Cone Axis=<0 0 1>
    Module.UseConeFalloff=false
    Module.UseVelDistribution=false
    Override Seed=false
    Rotation Mode=::NewEnumerator3
    Rotation Quaternion:Quat4f=$Particles.SkeletalMeshLocation.SampledRotation
    Velocity Mode=::NewEnumerator2
    Velocity Speed=&Multiply_Float(A=&RandomRangeFloat(Minimum=100, Maximum=90), B=$User.Intensity)
   ) {module=AddVelocity001}
   &AddVelocity(
    Module.Use Falloff Distance=false
    Origin Offset=<-10 0 0>
    Override Seed=false
    Velocity Mode=::NewEnumerator1
    Velocity Speed=&MakeCustomFloatFromBool(
     Bool=&SetBoolByFloatComparison(A=&VectorLength(Vector=$User.OwnerLocalVelocity), B=300)
     True Float=0
     False Float=&Multiply_Float(A=&RandomRangeFloat(Minimum=25, Maximum=35), B=$User.Intensity)
    )
   ) {enabled=false}
   &CameraOffset(Camera Offset Amount=2)
   /SetVariables(Particles.SpriteSize=&Multiply_Vector2D_ByFloat(
     Float=&FloatFromCurve(
      FloatCurve=DataInterface() {propertyValues {Curve {keys [<at 0 1>,<at 0.4 1>,<at 0.5043 0.2147>,<at 1.0029 0.1882>]}}}
      CurveIndex=$Emitter.NormalizedLoopAge
     )
     Vector2D=$Particles.SpriteSize
    )) {isSetParameters=true}
  ]
  particleUpdateScript [
   &SkeletalMeshLocation(
    Bone / Socket Index=1
    MaskSpawnGroup=false
    Mesh Sampling Type=::NewEnumerator5
    Orientation Sampling=::NewEnumerator1
    Position Sampling=::NewEnumerator2
    Skeletal Mesh=$System.Skeletal Mesh
    Socket Sampling Mode=::NewEnumerator5
   ) {module=SkeletalMeshLocation001}
   &ParticleState()
   &GravityForce(Gravity=<0 0 -350>)
   &BlendSpace("Position Blend (World <-> Local)":NiagaraFloat=&MakeCustomFloatFromBool(
     Bool=&SetBoolByFloatComparison(A=&VectorLength(Vector=$Engine.Owner.Velocity), B=300)
     True Float=1
     False Float=0
     Translation Contribution:NiagaraFloat=1
     Rotation Contribution:NiagaraFloat=1
    )) {module=BlendSpace001}
   &ScaleSpriteSize(
    Uniform Curve Scale=0.7
    Uniform Curve Sprite Scale=DataInterface() {propertyValues {Curve {
       defaultValue=0
       keys [
        {arriveTangent=1.0123, interpMode=RCIM_Cubic, leaveTangent=1.0123, tangentMode=RCTM_User, time=-0.0055, value=0.2443}
        {interpMode=RCIM_Cubic, tangentMode=RCTM_User, time=1.0092, value=0.8093}
       ]
      }}}
   )
   &ScaleColor(
    Scale Alpha=&FloatFromCurve(
     FloatCurve=DataInterface() {propertyValues {Curve {defaultValue=0.3, keys [{tangentMode=RCTM_User, time=0, value=1}, {tangentMode=RCTM_User, time=1, value=0}]}}}
     Scale Curve=0.7
    )
    ScaleA=true
    ScaleRGB=true
   )
   &SolveForcesAndVelocity(Clamp Velocity=false, Limit Acceleration=false)
   &DynamicMaterialParameters(
    Index 0 Param 1=&FN_FrontEndFloatModifier(
     Gameplay=&FloatFromCurve(FloatCurve=DataInterface() {propertyValues {Curve {keys [
          {arriveTangent=-0.0467, interpMode=RCIM_Cubic, leaveTangent=-0.0467, tangentMode=RCTM_User, time=0.1, value=1}
          {arriveTangent=-4.9722, tangentMode=RCTM_User, time=1, value=0}
         ]}}})
     Front End=&FloatFromCurve(FloatCurve=DataInterface() {propertyValues {Curve {keys [
          {arriveTangent=-0.0467, interpMode=RCIM_Cubic, leaveTangent=-0.0467, tangentMode=RCTM_User, time=0.1, value=1}
          {arriveTangent=-2.0505, interpMode=RCIM_Cubic, leaveTangent=-2.0505, tangentMode=RCTM_User, time=0.7044, value=0.6698}
          {arriveTangent=0.0025, tangentMode=RCTM_User, time=1, value=0}
         ]}}})
    )
    Index 0 Param 2=0.68
    Module.Index0Float0Write=true
    Module.Index0Float1Write=true
    Module.Index0Float2Write=true
    Module.Index0Float3Write=true
   )
   &KillParticlesInVolume(
    Box Size:Vector3f=<20 20 6>
    Invert Volume=true
    Kill Shape=::NewEnumerator1
    Volume Origin:NiagaraPosition=$Particles.SkeletalMeshLocation001.SampledPosition
   )
  ]
  rendererClasses [&NiagaraSpriteRendererProperties]
  renderers [{rendererClass=&NiagaraSpriteRendererProperties}]
 }
 New_Splashes001 {
  bEnabled=false
  emitterUpdateScript [
   &EmitterState(Life Cycle Mode=::NewEnumerator1, Loop Behavior=::NewEnumerator1, Loop Duration=0.5, UseLoopDelay=false)
   &SpawnRate(SpawnRate=5000, Use Spawn Probability=false)
  ]
  particleSpawnScript [
   &SkeletalMeshLocation(
    Bone / Socket Index=1
    MaskSpawnGroup=false
    Mesh Sampling Type=::NewEnumerator5
    Orientation Sampling=::NewEnumerator1
    Position Sampling=::NewEnumerator2
    Skeletal Mesh=$System.Skeletal Mesh
    Socket Sampling Mode=::NewEnumerator5
   )
   &InitializeParticle(
    Color=<rgba 1 0 0 1>
    Lifetime=2
    Sprite Rotation Mode=::NewEnumerator2
    Sprite Size Mode=::NewEnumerator1
    Sprite UV Mode=::NewEnumerator3
    Uniform Sprite Size Max=2
    Uniform Sprite Size Min=1
    UsePositionOffset=false
    Write Material Random=true
   )
   &SubUVAnimation(
    "Sprite Renderer "=DataInterface() {propertyValues {SpriteRenderer=&self:New_Splashes001_0.NiagaraSpriteRendererProperties_1}}
    SubUV Animation Mode:ENiagara_SubUVLookupModeV2=::NewEnumerator2
    UseEndFrame=false
    UseStartFrame=false
   )
   &ShapeLocation(
    Box Size=<20 20 20>
    MaskSpawnGroup=false
    Override Seed=false
    Shape Origin=$Particles.SkeletalMeshLocation.SampledPosition
    Shape Primitive=::NewEnumerator2
    Surface Only=false
    UseOwnerScale=true
   ) {module=ShapeLocation001}
   &AddVelocity(
    Cone Axis=&RotateVectorByQuaternion(Vector To Rotate=<0 1 0>, Quaternion=$Particles.SkeletalMeshLocation.SampledRotation)
    Module.UseConeFalloff=false
    Module.UseVelDistribution=false
    Override Seed=false
    Velocity Mode=::NewEnumerator2
    Velocity Speed=&Multiply_Float(A=&RandomRangeFloat(Minimum=40, Maximum=70), B=$User.Intensity)
   ) {enabled=false, module=AddVelocity001}
   &AddVelocity(
    Module.Use Falloff Distance=false
    Origin Offset=<-10 0 0>
    Override Seed=false
    Velocity Mode=::NewEnumerator1
    Velocity Speed=&MakeCustomFloatFromBool(
     Bool=&SetBoolByFloatComparison(A=&VectorLength(Vector=$User.OwnerLocalVelocity), B=300)
     True Float=0
     False Float=&Multiply_Float(A=&RandomRangeFloat(Minimum=25, Maximum=35), B=$User.Intensity)
    )
   ) {enabled=false}
   &CameraOffset(Camera Offset Amount=&RandomRangeFloat(
     Minimum=3
     Maximum=&MakeCustomFloatFromBool(
      Bool=&SetBoolByFloatComparison(A=&VectorLength(Vector=$User.OwnerLocalVelocity), B=20)
      True Float=10
      False Float=6
     )
    ))
  ]
  particleUpdateScript [
   &SkeletalMeshLocation(
    Bone / Socket Index=1
    Bone Sampling Mode=::NewEnumerator4
    MaskSpawnGroup=false
    Orientation Sampling=::NewEnumerator1
    Position Sampling=::NewEnumerator2
    Sampled Position Offset=<0 0 180>
    Sampled Position Offset Coordinate Space:ENiagaraSkelMeshTransforms=::NewEnumerator1
    Skeletal Mesh=$System.Skeletal Mesh
   ) {enabled=false, module=SkeletalMeshLocation001}
   &ParticleState()
   &GravityForce(Gravity=<0 0 -120>) {enabled=false}
   &BlendSpace("Position Blend (World <-> Local)":NiagaraFloat=&MakeCustomFloatFromBool(
     Bool=&SetBoolByFloatComparison(A=&VectorLength(Vector=$Engine.Owner.Velocity), B=300)
     True Float=1
     False Float=0
     Translation Contribution:NiagaraFloat=1
     Rotation Contribution:NiagaraFloat=1
    )) {enabled=false, module=BlendSpace001}
   &ScaleSpriteSize(
    Uniform Curve Scale=0.7
    Uniform Curve Sprite Scale=DataInterface() {propertyValues {Curve {
       defaultValue=0
       keys [{interpMode=RCIM_Cubic, tangentMode=RCTM_User, time=0, value=0}, {interpMode=RCIM_Cubic, tangentMode=RCTM_User, time=0.5, value=1}]
      }}}
   ) {enabled=false}
   &ScaleColor(
    Scale Alpha=&FloatFromCurve(
     FloatCurve=DataInterface() {propertyValues {Curve {defaultValue=0.3, keys [{tangentMode=RCTM_User, time=0, value=1}, {tangentMode=RCTM_User, time=1, value=0}]}}}
     Scale Curve=0.7
    )
    ScaleA=true
    ScaleRGB=true
   ) {enabled=false}
   &SolveForcesAndVelocity(Clamp Velocity=false, Limit Acceleration=false) {enabled=false}
   &DynamicMaterialParameters(
    Index 0 Param 1=&FN_FrontEndFloatModifier(
     Gameplay=&FloatFromCurve(FloatCurve=DataInterface() {propertyValues {Curve {keys [
          {arriveTangent=-0.0467, interpMode=RCIM_Cubic, leaveTangent=-0.0467, tangentMode=RCTM_User, time=0.1, value=1}
          {arriveTangent=-4.9722, tangentMode=RCTM_User, time=1, value=0}
         ]}}})
     Front End=&FloatFromCurve(FloatCurve=DataInterface() {propertyValues {Curve {keys [
          {arriveTangent=-0.0467, interpMode=RCIM_Cubic, leaveTangent=-0.0467, tangentMode=RCTM_User, time=0.1, value=1}
          {arriveTangent=-2.0505, interpMode=RCIM_Cubic, leaveTangent=-2.0505, tangentMode=RCTM_User, time=0.7044, value=0.6698}
          {arriveTangent=0.0025, tangentMode=RCTM_User, time=1, value=0}
         ]}}})
    )
    Index 0 Param 2=0.68
    Module.Index0Float0Write=true
    Module.Index0Float1Write=true
    Module.Index0Float2Write=true
    Module.Index0Float3Write=true
   ) {enabled=false}
   &KillParticlesInVolume(Sphere Radius=10, Volume Origin:NiagaraPosition=$Particles.SkeletalMeshLocation001.SampledPosition) {enabled=false}
  ]
  rendererClasses [&NiagaraSpriteRendererProperties]
  renderers [
   {rendererClass=&NiagaraSpriteRendererProperties}
   {rendererClass=&NiagaraSpriteRendererProperties, rendererIndex=1}
  ]
 }
 Speed_Lines {
  emitterUpdateScript [
   &EmitterState(
    Delay First Loop Only=true
    Life Cycle Mode=::NewEnumerator1
    Loop Behavior=::NewEnumerator2
    Loop Count=5
    Loop Delay=2.37
    Loop Duration=0.2
    UseLoopDelay=true
   )
   &SpawnBurst_Instantaneous(Spawn Count=10, Use Spawn Probability=false, UseLoopCountLimit=false)
   &SpawnBurst_Instantaneous(Spawn Count=10, Spawn Group=1, Use Spawn Probability=false, UseLoopCountLimit=false) {module=SpawnBurst_Instantaneous001}
  ]
  particleSpawnScript [
   &InitializeParticle(
    Color=&RandomRangeLinearColor(Minimum=<rgba 0 0 0 10>, Maximum=<rgba 0.0521 0.0521 0.0521 15>)
    Lifetime Max=0.14
    Lifetime Min=0.095
    Lifetime Mode=::NewEnumerator1
    Position Mode=::NewEnumerator1
    Sprite Size=<4 13>
    Sprite Size Mode=::NewEnumerator4
    UsePositionOffset=false
    Write Material Random=true
   )
   &SkeletalMeshLocation(
    Bone / Socket Index=1
    MaskSpawnGroup=false
    Mesh Sampling Type=::NewEnumerator5
    Position Sampling=::NewEnumerator2
    Sampled Position Offset=<7 2.5 0>
    Sampled Position Offset Coordinate Space:ENiagaraSkelMeshTransforms=::NewEnumerator2
    Skeletal Mesh=$System.Skeletal Mesh
    Socket Sampling Mode=::NewEnumerator5
   )
   &ShapeLocation(
    Box Size=<10 15 10>
    MaskSpawnGroup=false
    Offset=&MaskVectorBySpawnGroup(Masked Vector=<7 0 0>, Passthrough Vector=<-7 0 0>)
    Offset Coordinate Space=::Local
    Offset Mode=::NewEnumerator0
    Override Seed=false
    Shape Origin=$Particles.SkeletalMeshLocation.SampledPosition
    Shape Primitive=::NewEnumerator2
    Surface Only=false
    UseOwnerScale=true
   )
   &AddVelocity(Velocity=&MaskVectorBySpawnGroup(Masked Vector=<-10 0 0>, Passthrough Vector=<10 0 0>)) {module=AddVelocity001}
   &CameraOffset(Camera Offset Amount=&FN_FrontEndFloatModifier(Gameplay=0, Front End=-30))
  ]
  particleUpdateScript [
   &ParticleState()
   &Drag(Drag=5, Use Linear Drag=true, Use Rotational Drag=false)
   &SolveForcesAndVelocity(Clamp Velocity=false, Limit Acceleration=false)
   &ScaleColor(
    Scale Alpha=&FloatFromCurve(FloatCurve=DataInterface() {propertyValues {Curve {keys [
         {interpMode=RCIM_Cubic, tangentMode=RCTM_User, time=0.1552, value=0}
         {interpMode=RCIM_Cubic, tangentMode=RCTM_User, time=0.5, value=1}
         {tangentMode=RCTM_User, time=1, value=0}
        ]}}})
    ScaleA=true
    ScaleRGB=true
   )
   &ScaleSpriteSize(
    Non-Uniform Curve Sprite Scale=DataInterface() {propertyValues {
      XCurve {keys [{tangentMode=RCTM_User, time=0, value=0.5}, {tangentMode=RCTM_User, time=1, value=1.5}]}
      YCurve {keys [{tangentMode=RCTM_User, time=0, value=1}, {tangentMode=RCTM_User, time=1, value=0.5}]}
     }}
    Scale Sprite Size Mode=::NewEnumerator3
   )
   &ScaleSpriteSize(
    Scale Sprite Size Mode=::NewEnumerator0
    Uniform Scale Factor=&SelectFloatFromArray(
     Direct Array Index=&MakeIntFromFloat(Float=$User.WorkoutSelect)
     Float Selection Array=DataInterface() {propertyValues {FloatData [1,1,1,0.8]}}
    )
   ) {module=ScaleSpriteSize001}
   /SetVariables(Particles.New Position:NiagaraPosition=&Add_Vector(
     A=Linked() {linkedVariable {name=Particles.Position, type=NiagaraPosition}}
     B=Linked() {linkedVariable {name=Engine.Emitter.SimulationPosition, type=NiagaraPosition}}
    )) {enabled=false, isSetParameters=true, module=SetVariables_#2}
  ]
  rendererClasses [&NiagaraSpriteRendererProperties]
  renderers [{rendererClass=&NiagaraSpriteRendererProperties}]
 }
 Speed_LinesNoRepeat {
  emitterUpdateScript [
   &EmitterState(Life Cycle Mode=::NewEnumerator1, Loop Behavior=::NewEnumerator1, Loop Duration=3, UseLoopDelay=false)
   &SpawnBurst_Instantaneous(Spawn Count=10, Spawn Group=1, Spawn Time=1.15, Use Spawn Probability=false, UseLoopCountLimit=false) {module=SpawnBurst_Instantaneous001}
   &SpawnBurst_Instantaneous(Spawn Count=10, Spawn Time=1.7, Use Spawn Probability=false, UseLoopCountLimit=false)
   &SpawnBurst_Instantaneous(Spawn Count=10, Spawn Group=2, Spawn Time=1.95, Use Spawn Probability=false, UseLoopCountLimit=false) {module=SpawnBurst_Instantaneous002}
  ]
  particleSpawnScript [
   &InitializeParticle(
    Color=&RandomRangeLinearColor(Minimum=<rgba 0 0 0 10>, Maximum=<rgba 0.0521 0.0521 0.0521 15>)
    Lifetime Max=0.14
    Lifetime Min=0.095
    Lifetime Mode=::NewEnumerator1
    Position Mode=::NewEnumerator1
    Sprite Size=<5 15>
    Sprite Size Mode=::NewEnumerator4
    UsePositionOffset=false
    Write Material Random=true
   )
   &SkeletalMeshLocation(
    Bone / Socket Index=1
    MaskSpawnGroup=false
    Mesh Sampling Type=::NewEnumerator5
    Position Sampling=::NewEnumerator2
    Sampled Position Offset=<0 2.5 0>
    Sampled Position Offset Coordinate Space:ENiagaraSkelMeshTransforms=::NewEnumerator2
    Skeletal Mesh=$System.Skeletal Mesh
    Socket Sampling Mode=::NewEnumerator5
   )
   &ShapeLocation(
    Box Size=<10 25 10>
    MaskSpawnGroup=false
    Offset=&MaskVectorBySpawnGroup(
     Masked Vector=<18 0 0>
     Passthrough Vector=&MaskVectorBySpawnGroup(Masked Vector=<9 0 0>, Passthrough Vector=<-7 0 0>, Spawn Group Mask=1)
    )
    Offset Coordinate Space=::Local
    Offset Mode=::NewEnumerator0
    Override Seed=false
    Shape Origin=$Particles.SkeletalMeshLocation.SampledPosition
    Shape Primitive=::NewEnumerator2
    Surface Only=false
    UseOwnerScale=true
   )
   &AddVelocity(Velocity=&MaskVectorBySpawnGroup(Masked Vector=<-10 0 0>, Passthrough Vector=<10 0 0>)) {module=AddVelocity001}
   &CameraOffset(Camera Offset Amount=&FN_FrontEndFloatModifier(Gameplay=0, Front End=-30))
  ]
  particleUpdateScript [
   &ParticleState()
   &Drag(Drag=5, Use Linear Drag=true, Use Rotational Drag=false)
   &SolveForcesAndVelocity(Clamp Velocity=false, Limit Acceleration=false)
   &ScaleColor(
    Scale Alpha=&FloatFromCurve(FloatCurve=DataInterface() {propertyValues {Curve {keys [
         {interpMode=RCIM_Cubic, tangentMode=RCTM_User, time=0.1552, value=0}
         {interpMode=RCIM_Cubic, tangentMode=RCTM_User, time=0.5, value=1}
         {tangentMode=RCTM_User, time=1, value=0}
        ]}}})
    ScaleA=true
    ScaleRGB=true
   )
   &ScaleSpriteSize(
    Non-Uniform Curve Sprite Scale=DataInterface() {propertyValues {
      XCurve {keys [{tangentMode=RCTM_User, time=0, value=0.5}, {tangentMode=RCTM_User, time=1, value=1.5}]}
      YCurve {keys [{tangentMode=RCTM_User, time=0, value=1}, {tangentMode=RCTM_User, time=1, value=0.5}]}
     }}
    Scale Sprite Size Mode=::NewEnumerator3
   )
   &ScaleSpriteSize(
    Scale Sprite Size Mode=::NewEnumerator0
    Uniform Scale Factor=&SelectFloatFromArray(
     Direct Array Index=&MakeIntFromFloat(Float=$User.WorkoutSelect)
     Float Selection Array=DataInterface() {propertyValues {FloatData [1,1,1,0.8]}}
    )
   ) {module=ScaleSpriteSize001}
   /SetVariables(Particles.New Position:NiagaraPosition=&Add_Vector(
     A=Linked() {linkedVariable {name=Particles.Position, type=NiagaraPosition}}
     B=Linked() {linkedVariable {name=Engine.Emitter.SimulationPosition, type=NiagaraPosition}}
    )) {enabled=false, isSetParameters=true, module=SetVariables_#2}
  ]
  rendererClasses [&NiagaraSpriteRendererProperties]
  renderers [{rendererClass=&NiagaraSpriteRendererProperties}]
 }
}
summary {
 systemName=NS_SulfurDean_Hear_Drink
 emitters [
  {emitterName=New_Splashes, rendererClasses [&NiagaraSpriteRendererProperties]}
  {bEnabled=false, emitterName=BladeTrailDark, rendererClasses [&NiagaraRibbonRendererProperties]}
  {bEnabled=false, emitterName=New_Splashes001, rendererClasses [&NiagaraSpriteRendererProperties]}
  {emitterName=Speed_Lines, rendererClasses [&NiagaraSpriteRendererProperties]}
  {emitterName=Speed_LinesNoRepeat, rendererClasses [&NiagaraSpriteRendererProperties]}
 ]
 userVariables [
  User.bIsFrontEnd {type=NiagaraBool}
  User.bIsFrontEndPreview {type=NiagaraBool}
  User.Intensity {type=NiagaraFloat, defaultValue=0.68, description=""}
  User.OwnerLocalVelocity {type=Vector3f, description="", defaultValue {struct=Vector3f, value=<0 0 0>}}
  User.RibbonIDOffset {type=NiagaraInt32}
  User.WorkoutSelect {type=NiagaraFloat, defaultValue=0, description=""}
 ]
}
